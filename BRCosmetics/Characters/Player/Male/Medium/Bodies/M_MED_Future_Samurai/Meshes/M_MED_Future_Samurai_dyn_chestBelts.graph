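# Physics asset for the dynamic chest, back, spine and front-bag belts on the male medium Future Samurai body.
# Box and capsule bodies form short left and right belt chains, with tight twist and moderate swing limits.
# Chains root on small MSTR anchor bodies or the bag AIM capsule, joined to spine_05 or spine_02 by free-angle constraints.
# Capsules on pelvis, spine and thigh twist bones approximate the torso and upper legs.
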PhysicsAsset /BRCosmetics/Characters/Player/Male/Medium/Bodies/M_MED_Future_Samurai/Meshes/M_MED_Future_Samurai_dyn_chestBelts
bodies {
 dyn_bagBelt_01_l [{shapeType=Box, extentX=5, extentY=3, extentZ=1.3, center=<4 0 0>}]
 dyn_bagBelt_01_r [{shapeType=Box, extentX=5, extentY=3, extentZ=1.3, center=<4 0 0>}]
 dyn_bagBelt_02_l [{shapeType=Box, extentX=9, extentY=3, extentZ=1, center=<4 0 0>}]
 dyn_bagBelt_02_r [{shapeType=Box, extentX=9, extentY=3, extentZ=1, center=<4 0 0>}]
 dyn_bagBelt_03_l [{shapeType=Box, extentX=5, extentY=5.5, extentZ=1, center=<2.5 0 0>}]
 dyn_bagBelt_03_r [{shapeType=Box, extentX=5, extentY=5.5, extentZ=1, center=<2.5 0 0>}]
 dyn_belts_bk_l [{shapeType=Capsule, length=2.75, radius=1.75, center=<6 0 0>, rotation=<rot 0 0 90>}]
 dyn_belts_bk_l_MSTR [{shapeType=Capsule, length=0.5, radius=0.5}]
 dyn_belts_bk_r [{shapeType=Capsule, length=2.75, radius=1.75, center=<6 0 0>, rotation=<rot 0 0 90>}]
 dyn_belts_bk_r_MSTR [{shapeType=Capsule, length=0.5, radius=0.5}]
 dyn_belts_ft_l [{shapeType=Box, extentX=8.5, extentY=6, extentZ=1.5, center=<5 0 0>}]
 dyn_belts_ft_l_MSTR [{shapeType=Capsule, length=0.5, radius=0.5}]
 dyn_belts_ft_r [{shapeType=Box, extentX=8.5, extentY=6, extentZ=1.5, center=<6 0 0>}]
 dyn_belts_ft_r_MSTR [{shapeType=Capsule, length=0.5, radius=0.5}]
 dyn_front_bag_AIM [{shapeType=Capsule, length=9, radius=4.75, center=<9.6787 0 -0.8922>, rotation=<rot 0 0 90>}]
 dyn_spine_belt_01_l [{shapeType=Box, extentX=5.5, extentY=3.5, extentZ=1.5, center=<4.5 0 0>}]
 dyn_spine_belt_01_r [{shapeType=Box, extentX=5.5, extentY=3.5, extentZ=1.5, center=<4.5 0 0>}]
 dyn_spine_belt_02_l [{shapeType=Box, extentX=6, extentY=6.5, extentZ=2.5, center=<3.5 0 0>}]
 dyn_spine_belt_02_r [{shapeType=Box, extentX=6, extentY=6.5, extentZ=2.5, center=<3.5 0 0>}]
 dyn_spine_belt_l_MSTR [{shapeType=Capsule, length=0.5, radius=0.5}]
 dyn_spine_belt_r_MSTR [{shapeType=Capsule, length=0.5, radius=0.5}]
 pelvis [
  {shapeType=Capsule, length=10.75, radius=13.75, center=<-3.2529 -0.3746 0>}
  {shapeType=Capsule, length=10.75, radius=13.75, center=<-7.6928 0.919 0>}
 ]
 spine_01 [{shapeType=Capsule, length=12.5, radius=12.75, center=<0 -0.526 0>}]
 spine_02 [{shapeType=Capsule, length=9, radius=13, center=<0 0.9817 0>}]
 spine_03 [{shapeType=Capsule, length=8, radius=14, center=<0 1.0345 0>}]
 spine_04 [{shapeType=Capsule, length=10, radius=13, center=<0.0001 -1.1048 0>}]
 spine_05 [{shapeType=Capsule, length=15.25, radius=11.5}]
 thigh_twist_01_l [{shapeType=Capsule, length=37.4685, radius=9.9641, rotation=<rot 90 0 0>}]
 thigh_twist_01_r [{shapeType=Capsule, length=37.4685, radius=9.9641, rotation=<rot 90 0 0>}]
}
constraints [
 {bone1Name=dyn_bagBelt_02_r, bone2Name=dyn_bagBelt_01_r, swing1LimitDegrees=45, swing2LimitDegrees=45, twistLimitDegrees=5}
 {bone1Name=dyn_bagBelt_03_r, bone2Name=dyn_bagBelt_02_r, swing1LimitDegrees=30, swing2LimitDegrees=45, twistLimitDegrees=5}
 {bone1Name=dyn_bagBelt_02_l, bone2Name=dyn_bagBelt_01_l, swing1LimitDegrees=45, swing2LimitDegrees=45, twistLimitDegrees=5}
 {bone1Name=dyn_bagBelt_03_l, bone2Name=dyn_bagBelt_02_l, swing1LimitDegrees=30, swing2LimitDegrees=45, twistLimitDegrees=5}
 {bone1Name=dyn_spine_belt_02_r, bone2Name=dyn_spine_belt_01_r, swing1LimitDegrees=30, swing2LimitDegrees=30, twistLimitDegrees=5}
 {bone1Name=dyn_spine_belt_02_l, bone2Name=dyn_spine_belt_01_l, swing1LimitDegrees=30, swing2LimitDegrees=30, twistLimitDegrees=5}
 {bone1Name=dyn_belts_bk_l, bone2Name=dyn_belts_bk_l_MSTR, swing1LimitDegrees=10, swing2LimitDegrees=20, twistLimitDegrees=2}
 {bone1Name=dyn_belts_ft_l, bone2Name=dyn_belts_ft_l_MSTR, swing1LimitDegrees=15, swing2LimitDegrees=20, twistLimitDegrees=5}
 {bone1Name=dyn_belts_bk_r, bone2Name=dyn_belts_bk_r_MSTR, swing1LimitDegrees=10, swing2LimitDegrees=20, twistLimitDegrees=2}
 {bone1Name=dyn_belts_ft_r, bone2Name=dyn_belts_ft_r_MSTR, swing1LimitDegrees=15, swing2LimitDegrees=20, twistLimitDegrees=5}
 {bone1Name=dyn_bagBelt_01_r, bone2Name=dyn_front_bag_AIM, swing1LimitDegrees=30, swing2LimitDegrees=15, twistLimitDegrees=5}
 {bone1Name=dyn_bagBelt_01_l, bone2Name=dyn_front_bag_AIM, swing1LimitDegrees=30, swing2LimitDegrees=15, twistLimitDegrees=5}
 {bone1Name=dyn_spine_belt_01_r, bone2Name=dyn_spine_belt_r_MSTR, swing1LimitDegrees=45, swing2LimitDegrees=30, twistLimitDegrees=5}
 {bone1Name=dyn_spine_belt_01_l, bone2Name=dyn_spine_belt_l_MSTR, swing1LimitDegrees=45, swing2LimitDegrees=30, twistLimitDegrees=5}
 {
  bone1Name=dyn_belts_ft_r_MSTR
  bone2Name=spine_05
  swing1LimitDegrees=45
  swing1Motion=Free
  swing2LimitDegrees=45
  swing2Motion=Free
  twistLimitDegrees=45
  twistMotion=Free
 }
 {
  bone1Name=dyn_belts_bk_r_MSTR
  bone2Name=spine_05
  swing1LimitDegrees=45
  swing1Motion=Free
  swing2LimitDegrees=45
  swing2Motion=Free
  twistLimitDegrees=45
  twistMotion=Free
 }
 {
  bone1Name=dyn_belts_ft_l_MSTR
  bone2Name=spine_05
  swing1LimitDegrees=45
  swing1Motion=Free
  swing2LimitDegrees=45
  swing2Motion=Free
  twistLimitDegrees=45
  twistMotion=Free
 }
 {
  bone1Name=dyn_belts_bk_l_MSTR
  bone2Name=spine_05
  swing1LimitDegrees=45
  swing1Motion=Free
  swing2LimitDegrees=45
  swing2Motion=Free
  twistLimitDegrees=45
  twistMotion=Free
 }
 {
  bone1Name=dyn_front_bag_AIM
  bone2Name=spine_05
  swing1LimitDegrees=45
  swing1Motion=Free
  swing2LimitDegrees=45
  swing2Motion=Free
  twistLimitDegrees=45
  twistMotion=Free
 }
 {
  bone1Name=dyn_spine_belt_r_MSTR
  bone2Name=spine_02
  swing1LimitDegrees=45
  swing1Motion=Free
  swing2LimitDegrees=45
  swing2Motion=Free
  twistLimitDegrees=45
  twistMotion=Free
 }
 {
  bone1Name=dyn_spine_belt_l_MSTR
  bone2Name=spine_02
  swing1LimitDegrees=45
  swing1Motion=Free
  swing2LimitDegrees=45
  swing2Motion=Free
  twistLimitDegrees=45
  twistMotion=Free
 }
 {
  bone1Name=spine_02
  bone2Name=spine_01
  swing1LimitDegrees=45
  swing1Motion=Free
  swing2LimitDegrees=45
  swing2Motion=Free
  twistLimitDegrees=45
  twistMotion=Free
 }
 {
  bone1Name=spine_03
  bone2Name=spine_02
  swing1LimitDegrees=45
  swing1Motion=Free
  swing2LimitDegrees=45
  swing2Motion=Free
  twistLimitDegrees=45
  twistMotion=Free
 }
 {
  bone1Name=spine_04
  bone2Name=spine_03
  swing1LimitDegrees=45
  swing1Motion=Free
  swing2LimitDegrees=45
  swing2Motion=Free
  twistLimitDegrees=45
  twistMotion=Free
 }
 {
  bone1Name=spine_05
  bone2Name=spine_04
  swing1LimitDegrees=45
  swing1Motion=Free
  swing2LimitDegrees=45
  swing2Motion=Free
  twistLimitDegrees=45
  twistMotion=Free
 }
 {
  bone1Name=spine_01
  bone2Name=pelvis
  swing1LimitDegrees=45
  swing1Motion=Free
  swing2LimitDegrees=45
  swing2Motion=Free
  twistLimitDegrees=45
  twistMotion=Free
 }
 {
  bone1Name=thigh_twist_01_l
  bone2Name=pelvis
  swing1LimitDegrees=45
  swing1Motion=Free
  swing2LimitDegrees=45
  swing2Motion=Free
  twistLimitDegrees=45
  twistMotion=Free
 }
 {
  bone1Name=thigh_twist_01_r
  bone2Name=pelvis
  swing1LimitDegrees=45
  swing1Motion=Free
  swing2LimitDegrees=45
  swing2Motion=Free
  twistLimitDegrees=45
  twistMotion=Free
 }
]
