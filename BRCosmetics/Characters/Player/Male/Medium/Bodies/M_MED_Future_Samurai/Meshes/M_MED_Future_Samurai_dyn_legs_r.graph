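# Physics asset for the Future Samurai male medium body that simulates dangling pants belt straps on both legs.
# Capsules on pelvis, spine, thighs and calves form body volumes, joined by free constraints.
# Each leg has front and back four-bone strap chains with swing and twist limits, hung from small master bodies on the thigh twist bone.
# Locked constraints join the front and back chain tips, closing each strap into a loop.
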PhysicsAsset /BRCosmetics/Characters/Player/Male/Medium/Bodies/M_MED_Future_Samurai/Meshes/M_MED_Future_Samurai_dyn_legs_r
bodies {
 calf_l [{shapeType=Capsule, length=10.75, radius=13.25, center=<-7.8864 -2.4269 0.7362>, rotation=<rot 89.3713 93.001 -86.9597>}]
 calf_r [{shapeType=Capsule, length=10.75, radius=13.25, center=<7.8864 2.4269 -0.7362>, rotation=<rot 89.3713 93.001 93.0403>}]
 calf_twist_02_l [{shapeType=Capsule, length=26.875, radius=12.375, center=<-9.799 -1.0328 1.2047>, rotation=<rot 84.2767 3.0769 -170.8146>}]
 calf_twist_02_r [{shapeType=Capsule, length=26.875, radius=12.375, center=<9.799 1.0328 -1.2047>, rotation=<rot 84.2767 3.0769 9.1854>}]
 dyn_pantsBelt_bk_01_l [{shapeType=Capsule, length=12.411, radius=3, center=<8.5137 0 1.8>, rotation=<rot 87.9579 0.0001 0.0001>}]
 dyn_pantsBelt_bk_01_r [{shapeType=Capsule, length=8.8281, radius=3, center=<6.8551 0 1.8>, rotation=<rot 90 180 -180>}]
 dyn_pantsBelt_bk_02_l [{shapeType=Capsule, length=18, radius=3, center=<6.9098 0 1.2>, rotation=<rot 87.5925 0.0002 0>}]
 dyn_pantsBelt_bk_02_r [{shapeType=Capsule, length=16, radius=3, center=<7.7268 0 2>, rotation=<rot 90 0 0>}]
 dyn_pantsBelt_bk_03_l [{shapeType=Capsule, length=18, radius=3, center=<6.4972 0 1.2>, rotation=<rot 90 0 0>}]
 dyn_pantsBelt_bk_03_r [{shapeType=Capsule, length=16, radius=3, center=<6.6167 0 2>, rotation=<rot 90 0 0>}]
 dyn_pantsBelt_bk_04_l [{shapeType=Capsule, length=18, radius=3, center=<5.7466 0 -0.7326>, rotation=<rot 82.7344 0.0001 0.0001>}]
 dyn_pantsBelt_bk_04_r [{shapeType=Capsule, length=16, radius=3, center=<4.9118 0 -1.0525>, rotation=<rot 77.9054 0 0>}]
 dyn_pantsBelt_bk_l_MSTR [{shapeType=Capsule, length=0.5, radius=0.5, rotation=<rot 0 -90 0>}]
 dyn_pantsBelt_bk_r_MSTR [{shapeType=Capsule, length=0.5, radius=0.5, rotation=<rot 0 -90 0>}]
 dyn_pantsBelt_ft_01_l [{shapeType=Capsule, length=17.4377, radius=3, center=<10.6653 0 3>, rotation=<rot 90 0 0>}]
 dyn_pantsBelt_ft_01_r [{shapeType=Capsule, length=10.8267, radius=3, center=<6.8551 0 2>, rotation=<rot 90 0 0>}]
 dyn_pantsBelt_ft_02_l [{shapeType=Capsule, length=18, radius=3, center=<7.7842 0 2>, rotation=<rot 90 0 0>}]
 dyn_pantsBelt_ft_02_r [{shapeType=Capsule, length=16, radius=3, center=<7.7268 0 2>, rotation=<rot 90 0 0>}]
 dyn_pantsBelt_ft_03_l [{shapeType=Capsule, length=18, radius=3, center=<4.7549 0 2>, rotation=<rot 90 0 0>}]
 dyn_pantsBelt_ft_03_r [{shapeType=Capsule, length=16, radius=3, center=<6.6167 0 2>, rotation=<rot 90 0 0>}]
 dyn_pantsBelt_ft_04_l [{shapeType=Capsule, length=18, radius=3, center=<3.4382 0 -0.0099>, rotation=<rot 89.8357 0.0013 0.0014>}]
 dyn_pantsBelt_ft_04_r [{shapeType=Capsule, length=16, radius=3, center=<5.0233 0 0>, rotation=<rot 90 0 0>}]
 dyn_pantsBelt_ft_l_MSTR [{shapeType=Capsule, length=0.5, radius=0.5, rotation=<rot 0 -90 0>}]
 dyn_pantsBelt_ft_r_MSTR [{shapeType=Capsule, length=0.5, radius=0.5, rotation=<rot 0 -90 0>}]
 pelvis [{shapeType=Capsule, length=9.75, radius=14}]
 spine_01 [{shapeType=Capsule, length=9.75, radius=12.75, center=<3.1665 0.1864 0>}]
 thigh_twist_01_l [{shapeType=Capsule, length=58.3813, radius=10.6056, center=<-4.6461 -0.7077 -0.8263>, rotation=<rot 90 -179.7762 -0.5302>}]
 thigh_twist_01_r [{shapeType=Capsule, length=57.1697, radius=11.4406, center=<4.9677 0.7119 0>, rotation=<rot 90 -179.8881 179.3579>}]
}
constraints [
 {bone1Name=dyn_pantsBelt_ft_02_r, bone2Name=dyn_pantsBelt_ft_01_r, swing1LimitDegrees=75, swing2LimitDegrees=75, twistLimitDegrees=15}
 {bone1Name=dyn_pantsBelt_ft_03_r, bone2Name=dyn_pantsBelt_ft_02_r, swing1LimitDegrees=75, swing2LimitDegrees=75, twistLimitDegrees=15}
 {bone1Name=dyn_pantsBelt_bk_01_r, bone2Name=dyn_pantsBelt_bk_r_MSTR, swing1LimitDegrees=30, swing2LimitDegrees=30, twistLimitDegrees=15}
 {bone1Name=dyn_pantsBelt_bk_02_r, bone2Name=dyn_pantsBelt_bk_01_r, swing1LimitDegrees=75, swing2LimitDegrees=75, twistLimitDegrees=15}
 {bone1Name=dyn_pantsBelt_bk_03_r, bone2Name=dyn_pantsBelt_bk_02_r, swing1LimitDegrees=75, swing2LimitDegrees=75, twistLimitDegrees=15}
 {
  bone1Name=dyn_pantsBelt_bk_r_MSTR
  bone2Name=thigh_twist_01_r
  swing1LimitDegrees=45
  swing1Motion=Free
  swing2LimitDegrees=45
  swing2Motion=Free
  twistLimitDegrees=45
  twistMotion=Free
 }
 {bone1Name=dyn_pantsBelt_ft_01_r, bone2Name=dyn_pantsBelt_ft_r_MSTR, swing1LimitDegrees=60, swing2LimitDegrees=75, twistLimitDegrees=15}
 {
  bone1Name=dyn_pantsBelt_ft_r_MSTR
  bone2Name=thigh_twist_01_r
  swing1LimitDegrees=45
  swing1Motion=Free
  swing2LimitDegrees=45
  swing2Motion=Free
  twistLimitDegrees=45
  twistMotion=Free
 }
 {bone1Name=dyn_pantsBelt_bk_04_r, bone2Name=dyn_pantsBelt_bk_03_r, swing1LimitDegrees=75, swing2LimitDegrees=75, twistLimitDegrees=15}
 {bone1Name=dyn_pantsBelt_ft_04_r, bone2Name=dyn_pantsBelt_ft_03_r, swing1LimitDegrees=75, swing2LimitDegrees=75, twistLimitDegrees=15}
 {
  bone1Name=dyn_pantsBelt_ft_04_r
  bone2Name=dyn_pantsBelt_bk_04_r
  swing1LimitDegrees=5
  swing1Motion=Locked
  swing2LimitDegrees=5
  swing2Motion=Locked
  twistLimitDegrees=5
  twistMotion=Locked
 }
 {
  bone1Name=calf_twist_02_r
  bone2Name=calf_r
  swing1LimitDegrees=45
  swing1Motion=Free
  swing2LimitDegrees=45
  swing2Motion=Free
  twistLimitDegrees=45
  twistMotion=Free
 }
 {bone1Name=dyn_pantsBelt_bk_01_l, bone2Name=dyn_pantsBelt_bk_l_MSTR, swing1LimitDegrees=30, swing2LimitDegrees=30, twistLimitDegrees=15}
 {bone1Name=dyn_pantsBelt_bk_02_l, bone2Name=dyn_pantsBelt_bk_01_l, swing1LimitDegrees=90, swing2LimitDegrees=90, twistLimitDegrees=15}
 {bone1Name=dyn_pantsBelt_bk_03_l, bone2Name=dyn_pantsBelt_bk_02_l, swing1LimitDegrees=90, swing2LimitDegrees=90, twistLimitDegrees=15}
 {bone1Name=dyn_pantsBelt_bk_04_l, bone2Name=dyn_pantsBelt_bk_03_l, swing1LimitDegrees=90, swing2LimitDegrees=90, twistLimitDegrees=15}
 {bone1Name=dyn_pantsBelt_ft_01_l, bone2Name=dyn_pantsBelt_ft_l_MSTR, swing1LimitDegrees=20, swing2LimitDegrees=20, twistLimitDegrees=15}
 {bone1Name=dyn_pantsBelt_ft_02_l, bone2Name=dyn_pantsBelt_ft_01_l, swing1LimitDegrees=75, swing2LimitDegrees=75, twistLimitDegrees=15}
 {bone1Name=dyn_pantsBelt_ft_03_l, bone2Name=dyn_pantsBelt_ft_02_l, swing1LimitDegrees=75, swing2LimitDegrees=75, twistLimitDegrees=15}
 {bone1Name=dyn_pantsBelt_ft_04_l, bone2Name=dyn_pantsBelt_ft_03_l, swing1LimitDegrees=75, swing2LimitDegrees=75, twistLimitDegrees=15}
 {
  bone1Name=dyn_pantsBelt_bk_l_MSTR
  bone2Name=thigh_twist_01_l
  swing1LimitDegrees=45
  swing1Motion=Free
  swing2LimitDegrees=45
  swing2Motion=Free
  twistLimitDegrees=45
  twistMotion=Free
 }
 {
  bone1Name=dyn_pantsBelt_ft_l_MSTR
  bone2Name=thigh_twist_01_l
  swing1LimitDegrees=45
  swing1Motion=Free
  swing2LimitDegrees=45
  swing2Motion=Free
  twistLimitDegrees=45
  twistMotion=Free
 }
 {
  bone1Name=dyn_pantsBelt_ft_04_l
  bone2Name=dyn_pantsBelt_bk_04_l
  swing1LimitDegrees=5
  swing1Motion=Locked
  swing2LimitDegrees=5
  swing2Motion=Locked
  twistLimitDegrees=5
  twistMotion=Locked
 }
 {
  bone1Name=calf_r
  bone2Name=pelvis
  swing1LimitDegrees=45
  swing1Motion=Free
  swing2LimitDegrees=45
  swing2Motion=Free
  twistLimitDegrees=45
  twistMotion=Free
 }
 {
  bone1Name=thigh_twist_01_r
  bone2Name=pelvis
  swing1LimitDegrees=45
  swing1Motion=Free
  swing2LimitDegrees=45
  swing2Motion=Free
  twistLimitDegrees=45
  twistMotion=Free
 }
 {
  bone1Name=spine_01
  bone2Name=pelvis
  swing1LimitDegrees=45
  swing1Motion=Free
  swing2LimitDegrees=45
  swing2Motion=Free
  twistLimitDegrees=45
  twistMotion=Free
 }
 {
  bone1Name=thigh_twist_01_l
  bone2Name=pelvis
  swing1LimitDegrees=45
  swing1Motion=Free
  swing2LimitDegrees=45
  swing2Motion=Free
  twistLimitDegrees=45
  twistMotion=Free
 }
 {
  bone1Name=calf_twist_02_l
  bone2Name=calf_l
  swing1LimitDegrees=45
  swing1Motion=Free
  swing2LimitDegrees=45
  swing2Motion=Free
  twistLimitDegrees=45
  twistMotion=Free
 }
 {
  bone1Name=calf_l
  bone2Name=pelvis
  swing1LimitDegrees=45
  swing1Motion=Free
  swing2LimitDegrees=45
  swing2Motion=Free
  twistLimitDegrees=45
  twistMotion=Free
 }
]
