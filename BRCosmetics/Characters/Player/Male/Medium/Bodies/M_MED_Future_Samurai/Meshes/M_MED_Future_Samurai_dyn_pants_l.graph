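# Physics asset for the Future Samurai male medium body's left pants leg, setting up its dynamic cloth-panel bones for physics.
# Capsules on the thigh and calf bones form the leg; front, back, inner and outer boxes form the panels.
# Each panel has tight angular limits to a tiny MSTR capsule.
# Inner and outer MSTRs rotate freely on the thigh twist, front on the calf; back MSTR has no leg constraint.
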PhysicsAsset /BRCosmetics/Characters/Player/Male/Medium/Bodies/M_MED_Future_Samurai/Meshes/M_MED_Future_Samurai_dyn_pants_l
bodies {
 calf_l [{shapeType=Capsule, length=17.25, radius=6.25, center=<-14.2533 -3.2563 -0.4527>, rotation=<rot 90 -3.4277 176.0191>}]
 calf_twist_02_l [{shapeType=Capsule, length=18.75, radius=4.75, center=<-20.1486 -0.2148 0.664>, rotation=<rot 86.3626 -10.2246 -179.9995>}]
 dyn_shorts_bk_l [{shapeType=Box, extentX=18, extentY=19, extentZ=2, center=<23.9947 0 0.0002>}]
 dyn_shorts_bk_l_MSTR [{shapeType=Capsule, length=1, radius=1}]
 dyn_shorts_ft_l [{shapeType=Box, extentX=15, extentY=21.5, extentZ=2, center=<11.2955 0 0.0001>}]
 dyn_shorts_ft_l_MSTR [{shapeType=Capsule, length=1, radius=1}]
 dyn_shorts_in_l [{shapeType=Box, extentX=16, extentY=17, extentZ=2, center=<18 0 0>}]
 dyn_shorts_in_l_MSTR [{shapeType=Capsule, length=1, radius=1}]
 dyn_shorts_out_l [{shapeType=Box, extentX=13.5, extentY=20.5, extentZ=2, center=<19.4696 0 -0.0002>}]
 dyn_shorts_out_l_MSTR [{shapeType=Capsule, length=1, radius=1}]
 thigh_twist_01_l [{shapeType=Capsule, length=15.5, radius=8.25, center=<-2.3666 -0.2131 -0.0001>, rotation=<rot 90 0 174.8544>}]
}
constraints [
 {bone1Name=dyn_shorts_ft_l, bone2Name=dyn_shorts_ft_l_MSTR, swing1LimitDegrees=5, swing2LimitDegrees=5, twistLimitDegrees=5}
 {
  bone1Name=dyn_shorts_ft_l_MSTR
  bone2Name=calf_l
  swing1LimitDegrees=45
  swing1Motion=Free
  swing2LimitDegrees=45
  swing2Motion=Free
  twistLimitDegrees=45
  twistMotion=Free
 }
 {
  bone1Name=calf_twist_02_l
  bone2Name=calf_l
  swing1LimitDegrees=45
  swing1Motion=Free
  swing2LimitDegrees=45
  swing2Motion=Free
  twistLimitDegrees=45
  twistMotion=Free
 }
 {
  bone1Name=dyn_shorts_out_l_MSTR
  bone2Name=thigh_twist_01_l
  swing1LimitDegrees=45
  swing1Motion=Free
  swing2LimitDegrees=45
  swing2Motion=Free
  twistLimitDegrees=45
  twistMotion=Free
 }
 {bone1Name=dyn_shorts_out_l, bone2Name=dyn_shorts_out_l_MSTR, swing1LimitDegrees=5, swing2LimitDegrees=5, twistLimitDegrees=5}
 {
  bone1Name=dyn_shorts_in_l_MSTR
  bone2Name=thigh_twist_01_l
  swing1LimitDegrees=45
  swing1Motion=Free
  swing2LimitDegrees=45
  swing2Motion=Free
  twistLimitDegrees=45
  twistMotion=Free
 }
 {bone1Name=dyn_shorts_in_l, bone2Name=dyn_shorts_in_l_MSTR, swing1LimitDegrees=5, swing2LimitDegrees=5, twistLimitDegrees=5}
 {bone1Name=dyn_shorts_bk_l, bone2Name=dyn_shorts_bk_l_MSTR, swing1LimitDegrees=5, swing2LimitDegrees=5, twistLimitDegrees=2}
]
