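# Physics asset for the right pant leg of the Future Samurai medium male player cosmetic body, giving the dynamic cloth bones physics shapes.
# Capsules cover the thigh and calf, while flat boxes form front, back, inner and outer panels.
# Each panel has tight angular limits to a tiny MSTR capsule.
# These MSTR bodies swing and twist freely from the thigh, except the front one, which is constrained to the calf.
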PhysicsAsset /BRCosmetics/Characters/Player/Male/Medium/Bodies/M_MED_Future_Samurai/Meshes/M_MED_Future_Samurai_dyn_pants_r
bodies {
 calf_r [{shapeType=Capsule, length=16.75, radius=6.75, center=<14.1599 2.6497 0.7585>, rotation=<rot 90 -10.024 -10.3989>}]
 calf_twist_02_r [{shapeType=Capsule, length=18.25, radius=5.25, center=<20.2357 0.8088 -0.0002>, rotation=<rot 90 0 3.5324>}]
 dyn_shorts_bk_r [{shapeType=Box, extentX=18, extentY=18.5, extentZ=2, center=<23.8693 0 0.5088>, rotation=<rot 2.4192 0 0>}]
 dyn_shorts_bk_r_MSTR [{shapeType=Capsule, length=1, radius=1}]
 dyn_shorts_ft_r [{shapeType=Box, extentX=18, extentY=18.5, extentZ=2, center=<10 0 0>}]
 dyn_shorts_ft_r_MSTR [{shapeType=Capsule, length=1, radius=1}]
 dyn_shorts_in_r [{shapeType=Box, extentX=18, extentY=15.5, extentZ=2, center=<18 -0.0001 0>}]
 dyn_shorts_in_r_MSTR [{shapeType=Capsule, length=1, radius=1}]
 dyn_shorts_out_r [{shapeType=Box, extentX=18, extentY=18.5, extentZ=2.5, center=<18 -0.0001 0>}]
 dyn_shorts_out_r_MSTR [{shapeType=Capsule, length=1, radius=1}]
 thigh_twist_01_r [{shapeType=Capsule, length=26.25, radius=8.5, center=<2.3357 0.5566 0.0001>, rotation=<rot 90 0 -5.1456>}]
}
constraints [
 {bone1Name=dyn_shorts_ft_r, bone2Name=dyn_shorts_ft_r_MSTR, swing1LimitDegrees=5, swing2LimitDegrees=5, twistLimitDegrees=2}
 {
  bone1Name=dyn_shorts_ft_r_MSTR
  bone2Name=calf_r
  swing1LimitDegrees=45
  swing1Motion=Free
  swing2LimitDegrees=45
  swing2Motion=Free
  twistLimitDegrees=45
  twistMotion=Free
 }
 {
  bone1Name=calf_twist_02_r
  bone2Name=calf_r
  swing1LimitDegrees=45
  swing1Motion=Free
  swing2LimitDegrees=45
  swing2Motion=Free
  twistLimitDegrees=45
  twistMotion=Free
 }
 {
  bone1Name=dyn_shorts_out_r_MSTR
  bone2Name=thigh_twist_01_r
  swing1LimitDegrees=45
  swing1Motion=Free
  swing2LimitDegrees=45
  swing2Motion=Free
  twistLimitDegrees=45
  twistMotion=Free
 }
 {bone1Name=dyn_shorts_out_r, bone2Name=dyn_shorts_out_r_MSTR, swing1LimitDegrees=5, swing2LimitDegrees=5, twistLimitDegrees=2}
 {
  bone1Name=dyn_shorts_in_r_MSTR
  bone2Name=thigh_twist_01_r
  swing1LimitDegrees=45
  swing1Motion=Free
  swing2LimitDegrees=45
  swing2Motion=Free
  twistLimitDegrees=45
  twistMotion=Free
 }
 {bone1Name=dyn_shorts_in_r, bone2Name=dyn_shorts_in_r_MSTR, swing1LimitDegrees=5, swing2LimitDegrees=5, twistLimitDegrees=2}
 {
  bone1Name=dyn_shorts_bk_r_MSTR
  bone2Name=thigh_twist_01_r
  swing1LimitDegrees=45
  swing1Motion=Free
  swing2LimitDegrees=45
  swing2Motion=Free
  twistLimitDegrees=45
  twistMotion=Free
 }
 {bone1Name=dyn_shorts_bk_r, bone2Name=dyn_shorts_bk_r_MSTR, swing1LimitDegrees=5, swing2LimitDegrees=5, twistLimitDegrees=2}
]
