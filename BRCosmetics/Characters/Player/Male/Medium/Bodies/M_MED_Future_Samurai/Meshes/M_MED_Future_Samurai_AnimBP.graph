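# Animation Blueprint for the Future Samurai body cosmetic, adding secondary motion to bags, belts and shorts.
# It copies the parent mesh pose, applies an additive reset pose, then chains constraints, look-ats, corrective bone transforms and four RigidBody simulations.
# EventGraph derives corrective alphas from limb poses, varies gravity for swim-sprinting and skydiving, and resets dynamics on movement transitions.
# Several node chains, including crotch look-ats, are disconnected from the output.
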AnimBlueprint /BRCosmetics/Characters/Player/Male/Medium/Bodies/M_MED_Future_Samurai/Meshes/M_MED_Future_Samurai_AnimBP
target_skeleton=&/M_MED_Future_Samurai_Skeleton
graphs [
 AnimGraph {nodes [
   "Constraint - dyn_back_bag_TGT (PPPPP)"!(
    ComponentPose
    Alpha=1
    ConstraintWeights_0=0.0773
    ConstraintWeights_1:float=0.5208
    ConstraintWeights_2:float=0.1578
    ConstraintWeights_3:float=0.0377
    ConstraintWeights_4:float=0.2064
   ) -> (Pose) {
    alphaBoolBlend {blendOption=Linear}
    boneToModify {boneName=dyn_back_bag_TGT}
    constraintSetup [
     {targetBone {boneName=pelvis}}
     {targetBone {boneName=spine_02}}
     {targetBone {boneName=spine_03}}
     {targetBone {boneName=spine_04}}
     {targetBone {boneName=spine_01}}
    ]
    constraintWeights [1,1,1,1,1]
   }
   n1 = Get isNotCrouching() -> (isNotCrouching:double)
   n2 = Look At - Bone: dyn_crotch_l(Alpha=@n1) {
    alphaBoolBlend {blendOption=Linear}
    boneToModify {boneName=dyn_crotch_l}
    lookAtLocation {x=0}
    lookAtTarget {boneReference {boneName=dyn_crotch_aim_l}}
    lookAt_Axis {axis=<1 0>}
    lookUp_Axis {axis {x=0, z=1}}
   }
   Look At - Bone: dyn_crotch_r(ComponentPose=@n2, Alpha=@n1) {
    alphaBoolBlend {blendOption=Linear}
    boneToModify {boneName=dyn_crotch_r}
    lookAtLocation {x=0}
    lookAtTarget {boneReference {boneName=dyn_crotch_aim_r}}
    lookAt_Axis {axis=<1 0>}
    lookUp_Axis {axis {x=0, z=1}}
   }
   n3 = Get calf_rot_r() -> (calf_rot_r:double)
   n4 = "Transform (Modify) Bone - Bone: dyn_shorts_bk_r_BASE"(Alpha=@n3, Translation="-14.000,0,0", Rotation="0,0,0", Scale="1,1,1") {
    rotationMode=BMM_Additive
    rotationSpace=BCS_BoneSpace
    translationMode=BMM_Additive
    translationSpace=BCS_BoneSpace
    alphaBoolBlend {blendOption=Linear}
    boneToModify {boneName=dyn_shorts_bk_r_BASE}
    translation {x=-14}
   }
   n5 = "Transform (Modify) Bone - Bone: dyn_shorts_bk_r_MSTR"(ComponentPose=@n4, Alpha=@n3, Translation="-2,0,0", Rotation="0,0,0", Scale="1,1,1") {
    rotationMode=BMM_Additive
    rotationSpace=BCS_BoneSpace
    translationMode=BMM_Additive
    translationSpace=BCS_BoneSpace
    alphaBoolBlend {blendOption=Linear}
    boneToModify {boneName=dyn_shorts_bk_r_MSTR}
    translation {x=-2}
   }
   n6 = "Transform (Modify) Bone - Bone: dyn_shorts_in_r_MSTR"(ComponentPose=@n5, Alpha=@n3, Translation="0,0,0", Rotation="0,-75,0", Scale="1,1,1") {
    rotationMode=BMM_Additive
    rotationSpace=BCS_BoneSpace
    translationMode=BMM_Additive
    translationSpace=BCS_BoneSpace
    alphaBoolBlend {blendOption=Linear}
    boneToModify {boneName=dyn_shorts_in_r_MSTR}
    rotation {yaw=-75}
   }
   n7 = "Transform (Modify) Bone - Bone: dyn_shorts_out_r_MSTR"(ComponentPose=@n6, Alpha=@n3, Translation="0,0,0", Rotation="0,75,0", Scale="1,1,1") {
    rotationMode=BMM_Additive
    rotationSpace=BCS_BoneSpace
    translationMode=BMM_Additive
    translationSpace=BCS_BoneSpace
    alphaBoolBlend {blendOption=Linear}
    boneToModify {boneName=dyn_shorts_out_r_MSTR}
    rotation {yaw=75}
   }
   n8 = Get calf_rot_l() -> (calf_rot_l:double)
   n9 = "Transform (Modify) Bone - Bone: dyn_shorts_bk_l_BASE"(ComponentPose=@n7, Alpha=@n8, Translation="-14,0,0", Rotation="0,-10,0", Scale="1,1,1") {
    rotationMode=BMM_Additive
    rotationSpace=BCS_BoneSpace
    translationMode=BMM_Additive
    translationSpace=BCS_BoneSpace
    alphaBoolBlend {blendOption=Linear}
    boneToModify {boneName=dyn_shorts_bk_l_BASE}
    rotation {yaw=-10}
    translation {x=-14}
   }
   n10 = "Transform (Modify) Bone - Bone: dyn_shorts_bk_l_MSTR"(ComponentPose=@n9, Alpha=@n8, Translation="-2,0,0", Rotation="0,-5,0", Scale="1,1,1") {
    rotationMode=BMM_Additive
    rotationSpace=BCS_BoneSpace
    translationMode=BMM_Additive
    translationSpace=BCS_BoneSpace
    alphaBoolBlend {blendOption=Linear}
    boneToModify {boneName=dyn_shorts_bk_l_MSTR}
    rotation {yaw=-5}
    translation {x=-2}
   }
   n11 = "Transform (Modify) Bone - Bone: dyn_shorts_in_l_MSTR"(ComponentPose=@n10, Alpha=@n8, Translation="0,0,0", Rotation="0,75,0", Scale="1,1,1") {
    rotationMode=BMM_Additive
    rotationSpace=BCS_BoneSpace
    translationMode=BMM_Additive
    translationSpace=BCS_BoneSpace
    alphaBoolBlend {blendOption=Linear}
    boneToModify {boneName=dyn_shorts_in_l_MSTR}
    rotation {yaw=75}
   }
   "Transform (Modify) Bone - Bone: dyn_shorts_out_l_MSTR"(ComponentPose=@n11, Alpha=@n8, Translation="0,0,0", Rotation="0,-75,0", Scale="1,1,1") {
    rotationMode=BMM_Additive
    rotationSpace=BCS_BoneSpace
    translationMode=BMM_Additive
    translationSpace=BCS_BoneSpace
    alphaBoolBlend {blendOption=Linear}
    boneToModify {boneName=dyn_shorts_out_l_MSTR}
    rotation {yaw=-75}
   }
   n12 = Get calf_rot_r()
   n13 = "Transform (Modify) Bone - Bone: dyn_shorts_bk_r_BASE"(Alpha=@n12, Translation="0,0,0", Rotation="0,0,0", Scale="1,1,1") {
    rotationMode=BMM_Additive
    rotationSpace=BCS_BoneSpace
    scaleMode=BMM_Additive
    scaleSpace=BCS_BoneSpace
    translationMode=BMM_Additive
    translationSpace=BCS_BoneSpace
    alphaBoolBlend {blendOption=Linear}
    boneToModify {boneName=dyn_shorts_bk_r_BASE}
   }
   n14 = "Transform (Modify) Bone - Bone: dyn_shorts_bk_01_r"(ComponentPose=@n13, Alpha=@n12, Translation="0,0,0", Rotation="0,0,0", Scale="0.5,1,1") {
    rotationMode=BMM_Additive
    rotationSpace=BCS_BoneSpace
    scaleMode=BMM_Additive
    scaleSpace=BCS_BoneSpace
    translationMode=BMM_Additive
    translationSpace=BCS_BoneSpace
    alphaBoolBlend {blendOption=Linear}
    boneToModify {boneName=dyn_shorts_bk_01_r}
    scale {x=0.5}
   }
   n15 = "Transform (Modify) Bone - Bone: dyn_shorts_bk_02_r"(ComponentPose=@n14, Alpha=@n12, Translation="0,0,0", Rotation="0,0,0", Scale="0.5,1,1") {
    rotationMode=BMM_Additive
    rotationSpace=BCS_BoneSpace
    scaleMode=BMM_Additive
    scaleSpace=BCS_BoneSpace
    translationMode=BMM_Additive
    translationSpace=BCS_BoneSpace
    alphaBoolBlend {blendOption=Linear}
    boneToModify {boneName=dyn_shorts_bk_02_r}
    scale {x=0.5}
   }
   n16 = "Transform (Modify) Bone - Bone: dyn_shorts_ft_r"(ComponentPose=@n15, Alpha=@n12, Translation="2.4,2,-5", Rotation="60,0,17", Scale="0.35,1,1") {
    rotationMode=BMM_Additive
    rotationSpace=BCS_BoneSpace
    scaleMode=BMM_Additive
    scaleSpace=BCS_BoneSpace
    translationMode=BMM_Additive
    translationSpace=BCS_BoneSpace
    alphaBoolBlend {blendOption=Linear}
    boneToModify {boneName=dyn_shorts_ft_r}
    rotation {pitch=60, roll=17}
    scale {x=0.35}
    translation=<2.4 2 -5>
   }
   n17 = "Transform (Modify) Bone - Bone: dyn_shorts_in_l"(ComponentPose=@n16, Alpha=@n12, Translation="2,0,0", Rotation="0,0,0", Scale="0.6,1,1") {
    rotationMode=BMM_Additive
    rotationSpace=BCS_BoneSpace
    scaleMode=BMM_Additive
    scaleSpace=BCS_BoneSpace
    translationMode=BMM_Additive
    translationSpace=BCS_BoneSpace
    alphaBoolBlend {blendOption=Linear}
    boneToModify {boneName=dyn_shorts_in_l}
    scale {x=0.6}
    translation {x=2}
   }
   n18 = "Transform (Modify) Bone - Bone: dyn_shorts_out_l"(ComponentPose=@n17, Alpha=@n12, Translation="2,0,0", Rotation="0,0,0", Scale="0.5,1,1") {
    rotationMode=BMM_Additive
    rotationSpace=BCS_BoneSpace
    scaleMode=BMM_Additive
    scaleSpace=BCS_BoneSpace
    translationMode=BMM_Additive
    translationSpace=BCS_BoneSpace
    alphaBoolBlend {blendOption=Linear}
    boneToModify {boneName=dyn_shorts_out_l}
    scale {x=0.5}
    translation {x=2}
   }
   n19 = Get calf_rot_l()
   n20 = "Transform (Modify) Bone - Bone: dyn_shorts_bk_l_BASE"(ComponentPose=@n18, Alpha=@n19, Translation="0,0,0", Rotation="0,0,0", Scale="0.8,1,1") {
    rotationMode=BMM_Additive
    rotationSpace=BCS_BoneSpace
    scaleMode=BMM_Additive
    scaleSpace=BCS_BoneSpace
    translationMode=BMM_Additive
    translationSpace=BCS_BoneSpace
    alphaBoolBlend {blendOption=Linear}
    boneToModify {boneName=dyn_shorts_bk_l_BASE}
    scale {x=0.8}
   }
   n21 = "Transform (Modify) Bone - Bone: dyn_shorts_bk_01_l"(ComponentPose=@n20, Alpha=@n19, Translation="0,0,-2", Rotation="0,5,0", Scale="0.6,1,1") {
    rotationMode=BMM_Additive
    rotationSpace=BCS_BoneSpace
    scaleMode=BMM_Additive
    scaleSpace=BCS_BoneSpace
    translationMode=BMM_Additive
    translationSpace=BCS_BoneSpace
    alphaBoolBlend {blendOption=Linear}
    boneToModify {boneName=dyn_shorts_bk_01_l}
    rotation {yaw=5}
    scale {x=0.6}
    translation {z=-2}
   }
   n22 = "Transform (Modify) Bone - Bone: dyn_shorts_bk_02_l"(ComponentPose=@n21, Alpha=@n19, Translation="0,0,0", Rotation="0,0,0", Scale="1,1,1") {
    rotationMode=BMM_Additive
    rotationSpace=BCS_BoneSpace
    scaleMode=BMM_Additive
    scaleSpace=BCS_BoneSpace
    translationMode=BMM_Additive
    translationSpace=BCS_BoneSpace
    alphaBoolBlend {blendOption=Linear}
    boneToModify {boneName=dyn_shorts_bk_02_l}
   }
   n23 = "Transform (Modify) Bone - Bone: dyn_shorts_ft_l"(ComponentPose=@n22, Alpha=@n19, Translation="2,2,-5", Rotation="60,0,0", Scale="0.5,0.7,1") {
    rotationMode=BMM_Additive
    rotationSpace=BCS_BoneSpace
    scaleMode=BMM_Additive
    scaleSpace=BCS_BoneSpace
    translationMode=BMM_Additive
    translationSpace=BCS_BoneSpace
    alphaBoolBlend {blendOption=Linear}
    boneToModify {boneName=dyn_shorts_ft_l}
    rotation {pitch=60}
    scale=<0.5 0.7>
    translation=<2 2 -5>
   }
   n24 = "Transform (Modify) Bone - Bone: dyn_shorts_in_l"(ComponentPose=@n23, Alpha=@n19, Translation="2,0,0", Rotation="0,0,0", Scale="0.6,1,1") {
    rotationMode=BMM_Additive
    rotationSpace=BCS_BoneSpace
    scaleMode=BMM_Additive
    scaleSpace=BCS_BoneSpace
    translationMode=BMM_Additive
    translationSpace=BCS_BoneSpace
    alphaBoolBlend {blendOption=Linear}
    boneToModify {boneName=dyn_shorts_in_l}
    scale {x=0.6}
    translation {x=2}
   }
   "Transform (Modify) Bone - Bone: dyn_shorts_out_l"(ComponentPose=@n24, Alpha=@n19, Translation="2,0,0", Rotation="0,0,0", Scale="0.5,1,1") {
    rotationMode=BMM_Additive
    rotationSpace=BCS_BoneSpace
    scaleMode=BMM_Additive
    scaleSpace=BCS_BoneSpace
    translationMode=BMM_Additive
    translationSpace=BCS_BoneSpace
    alphaBoolBlend {blendOption=Linear}
    boneToModify {boneName=dyn_shorts_out_l}
    scale {x=0.5}
    translation {x=2}
   }
   n25 = Copy Pose From Mesh() {bUseAttachedParent=true}
   n26 = "Sequence Player (additive) 'M_MED_Future_Samurai_Skeleton_Reset_Pose'"() {sequence=&/M_MED_Future_Samurai_Skeleton_Reset_Pose}
   n27 = "Constraint - dyn_back_bag_MSTR (PPPP)"!(
    ComponentPose=Local To Component(LocalPose=Apply Additive(Base=@n25, Additive=@n26, Alpha=0.8) {alphaBoolBlend {blendOption=Linear}})
    Alpha=1
    ConstraintWeights_0=0.0074
    ConstraintWeights_1:float=0.0105
    ConstraintWeights_2:float=0.203
    ConstraintWeights_3:float=0.7792
   ) -> (Pose) {
    alphaBoolBlend {blendOption=Linear}
    boneToModify {boneName=dyn_back_bag_MSTR}
    constraintSetup [{targetBone {boneName=spine_02}}, {targetBone {boneName=spine_03}}, {targetBone {boneName=spine_04}}, {targetBone {boneName=spine_05}}]
    constraintWeights [1,1,1,1]
   }
   n28 = Look At - Bone: dyn_back_bag_AIM(ComponentPose=@n27, Alpha=1) {
    bUseLookUpAxis=true
    alphaBoolBlend {blendOption=Linear}
    boneToModify {boneName=dyn_back_bag_AIM}
    lookAtLocation {x=50}
    lookAtTarget {boneReference {boneName=dyn_back_bag_TGT}}
    lookAt_Axis {axis=<1 0>}
    lookUp_Axis {axis=<0 1>}
   }
   n29 = "Constraint - dyn_front_bag_MSTR (PP)"!(ComponentPose=@n28, Alpha=1, ConstraintWeights_0=0.2186, ConstraintWeights_1:float=0.7814) -> (Pose) {
    alphaBoolBlend {blendOption=Linear}
    boneToModify {boneName=dyn_front_bag_MSTR}
    constraintSetup [{targetBone {boneName=spine_04}}, {targetBone {boneName=spine_05}}]
    constraintWeights [1,1]
   }
   n30 = "Constraint - dyn_front_bag_TGT (PPPP)"!(
    ComponentPose=@n29
    Alpha=1
    ConstraintWeights_0=0.3187
    ConstraintWeights_1:float=0.2827
    ConstraintWeights_2:float=0.3219
    ConstraintWeights_3:float=0.0768
   ) -> (Pose) {
    alphaBoolBlend {blendOption=Linear}
    boneToModify {boneName=dyn_front_bag_TGT}
    constraintSetup [{targetBone {boneName=spine_02}}, {targetBone {boneName=spine_03}}, {targetBone {boneName=spine_04}}, {targetBone {boneName=spine_05}}]
    constraintWeights [1,1,1,1]
   }
   n31 = Look At - Bone: dyn_front_bag_AIM(ComponentPose=@n30, Alpha=1) {
    bUseLookUpAxis=true
    alphaBoolBlend {blendOption=Linear}
    boneToModify {boneName=dyn_front_bag_AIM}
    lookAtLocation {x=0}
    lookAtTarget {boneReference {boneName=dyn_front_bag_TGT}}
    lookAt_Axis {axis=<1 0>}
    lookUp_Axis {axis=<0 1>}
   }
   n32 = "Constraint - dyn_belts_bk_r_MSTR (PP)"!(ComponentPose=@n31, Alpha=1, ConstraintWeights_0=0.5, ConstraintWeights_1:float=0.5) -> (Pose) {
    alphaBoolBlend {blendOption=Linear}
    boneToModify {boneName=dyn_belts_bk_r_MSTR}
    constraintSetup [{targetBone {boneName=clavicle_r}}, {targetBone {boneName=spine_05}}]
    constraintWeights [1,1]
   }
   n33 = "Constraint - dyn_belts_bk_l_MSTR (PP)"!(ComponentPose=@n32, Alpha=1, ConstraintWeights_0=0.5, ConstraintWeights_1:float=0.5) -> (Pose) {
    alphaBoolBlend {blendOption=Linear}
    boneToModify {boneName=dyn_belts_bk_l_MSTR}
    constraintSetup [{targetBone {boneName=spine_05}}, {targetBone {boneName=clavicle_l}}]
    constraintWeights [1,1]
   }
   n34 = "Constraint - dyn_belts_ft_r_MSTR (PPPP)"!(
    ComponentPose=@n33
    Alpha=1
    ConstraintWeights_0=0.45
    ConstraintWeights_1:float=0
    ConstraintWeights_2:float=0
    ConstraintWeights_3:float=0.55
   ) -> (Pose) {
    alphaBoolBlend {blendOption=Linear}
    boneToModify {boneName=dyn_belts_ft_r_MSTR}
    constraintSetup [
     {targetBone {boneName=clavicle_r}}
     {targetBone {boneName=upperarm_twist_01_r}}
     {targetBone {boneName=spine_04}}
     {targetBone {boneName=spine_05}}
    ]
    constraintWeights [1,1,1,1]
   }
   n35 = "Constraint - dyn_belts_ft_l_MSTR (PPPP)"!(
    ComponentPose=@n34
    Alpha=1
    ConstraintWeights_0=0
    ConstraintWeights_1:float=0
    ConstraintWeights_2:float=0.55
    ConstraintWeights_3:float=0.45
   ) -> (Pose) {
    alphaBoolBlend {blendOption=Linear}
    boneToModify {boneName=dyn_belts_ft_l_MSTR}
    constraintSetup [
     {targetBone {boneName=upperarm_twist_01_l}}
     {targetBone {boneName=spine_04}}
     {targetBone {boneName=spine_05}}
     {targetBone {boneName=clavicle_l}}
    ]
    constraintWeights [1,1,1,1]
   }
   n36 = "Constraint - dyn_spine_belt_r_MSTR (PPPPP)"!(
    ComponentPose=@n35
    Alpha=1
    ConstraintWeights_0=0.0773
    ConstraintWeights_1:float=0.4793
    ConstraintWeights_2:float=0.1994
    ConstraintWeights_3:float=0.0377
    ConstraintWeights_4:float=0.2064
   ) -> (Pose) {
    alphaBoolBlend {blendOption=Linear}
    boneToModify {boneName=dyn_spine_belt_r_MSTR}
    constraintSetup [
     {targetBone {boneName=pelvis}}
     {targetBone {boneName=spine_02}}
     {targetBone {boneName=spine_03}}
     {targetBone {boneName=spine_04}}
     {targetBone {boneName=spine_01}}
    ]
    constraintWeights [1,1,1,1,1]
   }
   n37 = "Constraint - dyn_spine_belt_l_MSTR (PPPPP)"!(
    ComponentPose=@n36
    Alpha=1
    ConstraintWeights_0=0.0772
    ConstraintWeights_1:float=0.479
    ConstraintWeights_2:float=0.1997
    ConstraintWeights_3:float=0.0377
    ConstraintWeights_4:float=0.2064
   ) -> (Pose) {
    alphaBoolBlend {blendOption=Linear}
    boneToModify {boneName=dyn_spine_belt_l_MSTR}
    constraintSetup [
     {targetBone {boneName=pelvis}}
     {targetBone {boneName=spine_02}}
     {targetBone {boneName=spine_03}}
     {targetBone {boneName=spine_04}}
     {targetBone {boneName=spine_01}}
    ]
    constraintWeights [1,1,1,1,1]
   }
   n38 = "Constraint - dyn_pantsBelt_bk_r_MSTR (PPP)"!(ComponentPose=@n37, Alpha=1, ConstraintWeights_0=0.4503, ConstraintWeights_1:float=0.0031, ConstraintWeights_2:float=0.5466) -> (Pose) {
    alphaBoolBlend {blendOption=Linear}
    boneToModify {boneName=dyn_pantsBelt_bk_r_MSTR}
    constraintSetup [{targetBone {boneName=pelvis}}, {targetBone {boneName=thigh_twist_01_l}}, {targetBone {boneName=thigh_twist_01_r}}]
    constraintWeights [1,1,1]
   }
   n39 = "Constraint - dyn_pantsBelt_ft_r_MSTR (PPPP)"!(
    ComponentPose=@n38
    Alpha=1
    ConstraintWeights_0=0.0027
    ConstraintWeights_1:float=0.053
    ConstraintWeights_2:float=0.0801
    ConstraintWeights_3:float=0.8642
   ) -> (Pose) {
    alphaBoolBlend {blendOption=Linear}
    boneToModify {boneName=dyn_pantsBelt_ft_r_MSTR}
    constraintSetup [
     {targetBone {boneName=thigh_r}}
     {targetBone {boneName=pelvis}}
     {targetBone {boneName=thigh_twist_01_l}}
     {targetBone {boneName=thigh_twist_01_r}}
    ]
    constraintWeights [1,1,1,1]
   }
   n40 = "Constraint - dyn_pantsBelt_bk_l_MSTR (PPP)"!(ComponentPose=@n39, Alpha=1, ConstraintWeights_0=0.5322, ConstraintWeights_1:float=0.4594, ConstraintWeights_2:float=0.0084) -> (Pose) {
    alphaBoolBlend {blendOption=Linear}
    boneToModify {boneName=dyn_pantsBelt_bk_l_MSTR}
    constraintSetup [{targetBone {boneName=pelvis}}, {targetBone {boneName=thigh_twist_01_l}}, {targetBone {boneName=thigh_twist_01_r}}]
    constraintWeights [1,1,1]
   }
   n41 = "Constraint - dyn_pantsBelt_ft_l_MSTR (PPP)"!(ComponentPose=@n40, Alpha=1, ConstraintWeights_0=0.1221, ConstraintWeights_1:float=0.723, ConstraintWeights_2:float=0.1549) -> (Pose) {
    alphaBoolBlend {blendOption=Linear}
    boneToModify {boneName=dyn_pantsBelt_ft_l_MSTR}
    constraintSetup [{targetBone {boneName=pelvis}}, {targetBone {boneName=thigh_twist_01_l}}, {targetBone {boneName=thigh_twist_01_r}}]
    constraintWeights [1,1,1]
   }
   n42 = "Constraint - dyn_shorts_in_l_MSTR (PP)"!(ComponentPose=@n41, Alpha=0, ConstraintWeights_0=0.5, ConstraintWeights_1:float=0.5) -> (Pose) {
    alphaBoolBlend {blendOption=Linear}
    boneToModify {boneName=dyn_shorts_in_l_MSTR}
    constraintSetup [{targetBone {boneName=thigh_twist_01_l}}, {targetBone {boneName=thigh_twist_01_r}}]
    constraintWeights [1,1]
   }
   n43 = "Constraint - dyn_shorts_in_r_MSTR (PP)"!(ComponentPose=@n42, Alpha=0, ConstraintWeights_0=0.5, ConstraintWeights_1:float=0.5) -> (Pose) {
    alphaBoolBlend {blendOption=Linear}
    boneToModify {boneName=dyn_shorts_in_r_MSTR}
    constraintSetup [{targetBone {boneName=thigh_twist_01_l}}, {targetBone {boneName=thigh_twist_01_r}}]
    constraintWeights [1,1]
   }
   n44 = "Transform (Modify) Bone - Bone: dyn_pantsBelt_ft_r_MSTR"(ComponentPose=@n43, Alpha=1, Translation="0,0,0", Rotation="0.000,20,0.000", Scale="1,1,1") {
    rotationMode=BMM_Additive
    rotationSpace=BCS_BoneSpace
    translationMode=BMM_Additive
    translationSpace=BCS_BoneSpace
    alphaBoolBlend {blendOption=Linear}
    boneToModify {boneName=dyn_pantsBelt_ft_r_MSTR}
    rotation {yaw=20}
   }
   n45 = "Transform (Modify) Bone - Bone: dyn_pantsBelt_bk_r_MSTR"(ComponentPose=@n44, Alpha=1, Translation="0,0,0", Rotation="0.000,-20,0.000", Scale="1,1,1") {
    rotationMode=BMM_Additive
    rotationSpace=BCS_BoneSpace
    translationMode=BMM_Additive
    translationSpace=BCS_BoneSpace
    alphaBoolBlend {blendOption=Linear}
    boneToModify {boneName=dyn_pantsBelt_bk_r_MSTR}
    rotation {yaw=-20}
   }
   n46 = "Transform (Modify) Bone - Bone: dyn_pantsBelt_ft_l_MSTR"(ComponentPose=@n45, Alpha=1, Translation="0,0,0", Rotation="0.000,-20,0.000", Scale="1,1,1") {
    rotationMode=BMM_Additive
    rotationSpace=BCS_BoneSpace
    translationMode=BMM_Additive
    translationSpace=BCS_BoneSpace
    alphaBoolBlend {blendOption=Linear}
    boneToModify {boneName=dyn_pantsBelt_ft_l_MSTR}
    rotation {yaw=-20}
   }
   n47 = "Transform (Modify) Bone - Bone: dyn_pantsBelt_bk_l_MSTR"(ComponentPose=@n46, Alpha=1, Translation="0,0,0", Rotation="0.000,20,0.000", Scale="1,1,1") {
    rotationMode=BMM_Additive
    rotationSpace=BCS_BoneSpace
    translationMode=BMM_Additive
    translationSpace=BCS_BoneSpace
    alphaBoolBlend {blendOption=Linear}
    boneToModify {boneName=dyn_pantsBelt_bk_l_MSTR}
    rotation {yaw=20}
   }
   n48 = Get thigh_up_r() -> (thigh_up_r:double)
   n49 = "Transform (Modify) Bone - Bone: dyn_pantsBelt_ft_r_MSTR"(ComponentPose=@n47, Alpha=@n48, Translation="6.031,1.935,3.000", Rotation="0.000,30,0.000", Scale="1,1,1") {
    rotationMode=BMM_Additive
    rotationSpace=BCS_BoneSpace
    translationMode=BMM_Additive
    translationSpace=BCS_BoneSpace
    alphaBoolBlend {blendOption=Linear}
    boneToModify {boneName=dyn_pantsBelt_ft_r_MSTR}
    rotation {yaw=30}
    translation=<6.031 1.935 3>
   }
   n50 = "Transform (Modify) Bone - Bone: dyn_pantsBelt_bk_r_MSTR"(ComponentPose=@n49, Alpha=@n48, Translation="0,0,0", Rotation="0.000,-30,0.000", Scale="1,1,1") {
    rotationMode=BMM_Additive
    rotationSpace=BCS_BoneSpace
    translationMode=BMM_Additive
    translationSpace=BCS_BoneSpace
    alphaBoolBlend {blendOption=Linear}
    boneToModify {boneName=dyn_pantsBelt_bk_r_MSTR}
    rotation {yaw=-30}
   }
   n51 = "Transform (Modify) Bone - Bone: dyn_pantsBelt_ft_l_MSTR"(ComponentPose=@n50, Alpha=Get thigh_up_l() -> (thigh_up_l:double), Translation="0,0,5", Rotation="0.000,-15,0.000", Scale="1,1,1") {
    rotationMode=BMM_Additive
    rotationSpace=BCS_BoneSpace
    translationMode=BMM_Additive
    translationSpace=BCS_BoneSpace
    alphaBoolBlend {blendOption=Linear}
    boneToModify {boneName=dyn_pantsBelt_ft_l_MSTR}
    rotation {yaw=-15}
    translation {z=5}
   }
   n52 = "Transform (Modify) Bone - Bone: dyn_spine_belt_l_MSTR"(ComponentPose=@n51, Alpha=Get thigh_up_l(), Translation="0,3.67,0", Rotation="10,0,10", Scale="1,1,1") {
    rotationMode=BMM_Additive
    rotationSpace=BCS_BoneSpace
    translationMode=BMM_Additive
    translationSpace=BCS_BoneSpace
    alphaBoolBlend {blendOption=Linear}
    boneToModify {boneName=dyn_spine_belt_l_MSTR}
    rotation {pitch=10, roll=10}
    translation {y=3.67}
   }
   n53 = "Transform (Modify) Bone - Bone: dyn_spine_belt_r_MSTR"(ComponentPose=@n52, Alpha=Get thigh_up_r(), Translation="0,-3.67,0", Rotation="10,0,-10", Scale="1,1,1") {
    rotationMode=BMM_Additive
    rotationSpace=BCS_BoneSpace
    translationMode=BMM_Additive
    translationSpace=BCS_BoneSpace
    alphaBoolBlend {blendOption=Linear}
    boneToModify {boneName=dyn_spine_belt_r_MSTR}
    rotation {pitch=10, roll=-10}
    translation {y=-3.67}
   }
   n54 = "Transform (Modify) Bone - Bone: dyn_belts_ft_r_MSTR"(ComponentPose=@n53, Alpha=Get ADS alpha() -> (ADS alpha:double), Translation="0,-2,0", Rotation="0,0,0", Scale="1,1,1") {
    rotationMode=BMM_Additive
    rotationSpace=BCS_BoneSpace
    translationMode=BMM_Additive
    translationSpace=BCS_BoneSpace
    alphaBoolBlend {blendOption=Linear}
    boneToModify {boneName=dyn_belts_ft_r_MSTR}
    translation {y=-2}
   }
   n55 = "Transform (Modify) Bone - Bone: dyn_belts_ft_l_MSTR"(ComponentPose=@n54, Alpha=Get arm_forward_l() -> (arm_forward_l:double), Translation="0,2,0.000", Rotation="0,0,0", Scale="1,1,1") {
    rotationMode=BMM_Additive
    rotationSpace=BCS_BoneSpace
    translationMode=BMM_Additive
    translationSpace=BCS_BoneSpace
    alphaBoolBlend {blendOption=Linear}
    boneToModify {boneName=dyn_belts_ft_l_MSTR}
    translation {y=2}
   }
   n56 = Get calf_rot_r()
   n57 = "Transform (Modify) Bone - Bone: dyn_shorts_bk_r_MSTR"(ComponentPose=@n55, Alpha=@n56, Translation="0,0,0", Rotation="90,0,0", Scale="1,1,1") {
    rotationMode=BMM_Additive
    rotationSpace=BCS_BoneSpace
    translationMode=BMM_Additive
    translationSpace=BCS_BoneSpace
    alphaBoolBlend {blendOption=Linear}
    boneToModify {boneName=dyn_shorts_bk_r_MSTR}
    rotation {pitch=90}
   }
   n58 = "Transform (Modify) Bone - Bone: dyn_shorts_out_r_MSTR"(ComponentPose=@n57, Alpha=@n56, Translation="0,0,0", Rotation="0,60,0.000", Scale="1,1,1") {
    rotationMode=BMM_Additive
    rotationSpace=BCS_BoneSpace
    translationMode=BMM_Additive
    translationSpace=BCS_BoneSpace
    alphaBoolBlend {blendOption=Linear}
    boneToModify {boneName=dyn_shorts_out_r_MSTR}
    rotation {yaw=60}
   }
   n59 = "Transform (Modify) Bone - Bone: dyn_shorts_in_r_MSTR"(ComponentPose=@n58, Alpha=@n56, Translation="0,0,0", Rotation="0,-40,0", Scale="1,1,1") {
    rotationMode=BMM_Additive
    rotationSpace=BCS_BoneSpace
    translationMode=BMM_Additive
    translationSpace=BCS_BoneSpace
    alphaBoolBlend {blendOption=Linear}
    boneToModify {boneName=dyn_shorts_in_r_MSTR}
    rotation {yaw=-40}
   }
   n60 = Get calf_rot_l()
   n61 = "Transform (Modify) Bone - Bone: dyn_shorts_bk_l_MSTR"(ComponentPose=@n59, Alpha=@n60, Translation="0,0,0", Rotation="90,0,0", Scale="1,1,1") {
    rotationMode=BMM_Additive
    rotationSpace=BCS_BoneSpace
    translationMode=BMM_Additive
    translationSpace=BCS_BoneSpace
    alphaBoolBlend {blendOption=Linear}
    boneToModify {boneName=dyn_shorts_bk_l_MSTR}
    rotation {pitch=90}
   }
   n62 = "Transform (Modify) Bone - Bone: dyn_shorts_out_l_MSTR"(ComponentPose=@n61, Alpha=@n60, Translation="0,0,0", Rotation="0,-60,0.000", Scale="1,1,1") {
    rotationMode=BMM_Additive
    rotationSpace=BCS_BoneSpace
    translationMode=BMM_Additive
    translationSpace=BCS_BoneSpace
    alphaBoolBlend {blendOption=Linear}
    boneToModify {boneName=dyn_shorts_out_l_MSTR}
    rotation {yaw=-60}
   }
   n63 = "Transform (Modify) Bone - Bone: dyn_shorts_in_l_MSTR"(ComponentPose=@n62, Alpha=@n60, Translation="0,0,0", Rotation="0,40,0", Scale="1,1,1") {
    rotationMode=BMM_Additive
    rotationSpace=BCS_BoneSpace
    translationMode=BMM_Additive
    translationSpace=BCS_BoneSpace
    alphaBoolBlend {blendOption=Linear}
    boneToModify {boneName=dyn_shorts_in_l_MSTR}
    rotation {yaw=40}
   }
   n64 = Get legs_rot_sides() -> (legs_rot_sides:double)
   n65 = "Transform (Modify) Bone - Bone: dyn_shorts_in_r_MSTR"(ComponentPose=@n63, Alpha=@n64, Translation="0,0,8", Rotation="-10,0.000,0.000", Scale="1,1,1") {
    rotationMode=BMM_Additive
    rotationSpace=BCS_BoneSpace
    translationMode=BMM_Additive
    translationSpace=BCS_BoneSpace
    alphaBoolBlend {blendOption=Linear}
    boneToModify {boneName=dyn_shorts_in_r_MSTR}
    rotation {pitch=-10}
    translation {z=8}
   }
   n66 = "Transform (Modify) Bone - Bone: dyn_shorts_in_l_MSTR"(ComponentPose=@n65, Alpha=@n64, Translation="0,0,8", Rotation="-10,0.000,0.000", Scale="1,1,1") {
    rotationMode=BMM_Additive
    rotationSpace=BCS_BoneSpace
    translationMode=BMM_Additive
    translationSpace=BCS_BoneSpace
    alphaBoolBlend {blendOption=Linear}
    boneToModify {boneName=dyn_shorts_in_l_MSTR}
    rotation {pitch=-10}
    translation {z=8}
   }
   n67 = Get gravity_override() -> (gravity_override:Vector)
   n68 = RigidBody!(ComponentPose=@n66, Alpha, OverrideWorldGravity:Vector=@n67, ExternalForce) -> (Pose) {
    bClampLinearTranslationLimitToRefPose=true
    bOverrideWorldGravity=true
    overridePhysicsAsset=&/M_MED_Future_Samurai_dyn_chestBelts
    simulationSpace=BaseBoneSpace
    alphaBoolBlend {blendOption=Linear}
    baseBoneRef {boneName=root}
    componentAppliedLinearAccClamp=<2000 2000 2000>
    componentLinearAccScale=<0.1 0.1 0.1>
    componentLinearVelScale=<0.1 0.1 0.1>
    overrideWorldGravity {z=-980}
   }
   n69 = RigidBody(ComponentPose=@n68, OverrideWorldGravity=@n67) {
    bClampLinearTranslationLimitToRefPose=true
    bOverrideWorldGravity=true
    overridePhysicsAsset=&/M_MED_Future_Samurai_dyn_legs_r
    simulationSpace=BaseBoneSpace
    alphaBoolBlend {blendOption=Linear}
    baseBoneRef {boneName=root}
    componentAppliedLinearAccClamp=<1500 1500 1500>
    componentLinearAccScale=<0.1 0.1 0.1>
    componentLinearVelScale=<0.1 0.1 0.1>
    overrideWorldGravity {z=-980}
   }
   n70 = RigidBody(ComponentPose=@n69, OverrideWorldGravity=@n67) {
    bClampLinearTranslationLimitToRefPose=true
    bOverrideWorldGravity=true
    overridePhysicsAsset=&/M_MED_Future_Samurai_dyn_pants_r
    simulationSpace=BaseBoneSpace
    alphaBoolBlend {blendOption=Linear}
    baseBoneRef {boneName=calf_r}
    componentAppliedLinearAccClamp=<1000 1000 1000>
    componentLinearAccScale=<0.2 0.2 0.2>
    componentLinearVelScale=<0.2 0.2 0.2>
    overrideWorldGravity {z=-980}
   }
   n71 = RigidBody(ComponentPose=@n70, OverrideWorldGravity=@n67) {
    bClampLinearTranslationLimitToRefPose=true
    bOverrideWorldGravity=true
    overridePhysicsAsset=&/M_MED_Future_Samurai_dyn_pants_l
    simulationSpace=BaseBoneSpace
    alphaBoolBlend {blendOption=Linear}
    baseBoneRef {boneName=calf_l}
    componentAppliedLinearAccClamp=<1000 1000 1000>
    componentLinearAccScale=<0.2 0.2 0.2>
    componentLinearVelScale=<0.2 0.2 0.2>
    overrideWorldGravity {z=-980}
   }
   n72 = Get calf_rot_r()
   n73 = "Transform (Modify) Bone - Bone: dyn_shorts_bk_r"(ComponentPose=@n71, Alpha=@n72, Translation="0,0,-1", Rotation="0,0,0", Scale="0.35,1,1") {
    rotationMode=BMM_Additive
    rotationSpace=BCS_BoneSpace
    scaleMode=BMM_Additive
    scaleSpace=BCS_BoneSpace
    translationMode=BMM_Additive
    translationSpace=BCS_BoneSpace
    alphaBoolBlend {blendOption=Linear}
    boneToModify {boneName=dyn_shorts_bk_r}
    scale {x=0.35}
    translation {z=-1}
   }
   n74 = "Transform (Modify) Bone - Bone: dyn_shorts_in_r"(ComponentPose=@n73, Alpha=@n72, Translation="0,0,0", Rotation="0,0,0", Scale="0.5,1,1") {
    rotationMode=BMM_Additive
    rotationSpace=BCS_BoneSpace
    scaleMode=BMM_Additive
    scaleSpace=BCS_BoneSpace
    translationMode=BMM_Additive
    translationSpace=BCS_BoneSpace
    alphaBoolBlend {blendOption=Linear}
    boneToModify {boneName=dyn_shorts_in_r}
    scale {x=0.5}
   }
   n75 = "Transform (Modify) Bone - Bone: dyn_shorts_out_r"(ComponentPose=@n74, Alpha=@n72, Translation="0,0,0", Rotation="0,0,0", Scale="0.5,1,1") {
    rotationMode=BMM_Additive
    rotationSpace=BCS_BoneSpace
    scaleMode=BMM_Additive
    scaleSpace=BCS_BoneSpace
    translationMode=BMM_Additive
    translationSpace=BCS_BoneSpace
    alphaBoolBlend {blendOption=Linear}
    boneToModify {boneName=dyn_shorts_out_r}
    scale {x=0.5}
   }
   n76 = "Transform (Modify) Bone - Bone: dyn_shorts_ft_r"(ComponentPose=@n75, Alpha=@n72, Translation="0,0,0", Rotation="-30,0,0", Scale="0.8,1,1") {
    rotationMode=BMM_Additive
    rotationSpace=BCS_BoneSpace
    scaleMode=BMM_Additive
    scaleSpace=BCS_BoneSpace
    translationMode=BMM_Additive
    translationSpace=BCS_BoneSpace
    alphaBoolBlend {blendOption=Linear}
    boneToModify {boneName=dyn_shorts_ft_r}
    rotation {pitch=-30}
    scale {x=0.8}
   }
   n77 = Get calf_rot_l()
   n78 = "Transform (Modify) Bone - Bone: dyn_shorts_bk_l"(ComponentPose=@n76, Alpha=@n77, Translation="0,0,0", Rotation="0,0,0", Scale="0.3,1,1") {
    rotationMode=BMM_Additive
    rotationSpace=BCS_BoneSpace
    scaleMode=BMM_Additive
    scaleSpace=BCS_BoneSpace
    translationMode=BMM_Additive
    translationSpace=BCS_BoneSpace
    alphaBoolBlend {blendOption=Linear}
    boneToModify {boneName=dyn_shorts_bk_l}
    scale {x=0.3}
   }
   n79 = "Transform (Modify) Bone - Bone: dyn_shorts_in_l"(ComponentPose=@n78, Alpha=@n77, Translation="0,0,0", Rotation="0,0,0", Scale="0.45,1,1") {
    rotationMode=BMM_Additive
    rotationSpace=BCS_BoneSpace
    scaleMode=BMM_Additive
    scaleSpace=BCS_BoneSpace
    translationMode=BMM_Additive
    translationSpace=BCS_BoneSpace
    alphaBoolBlend {blendOption=Linear}
    boneToModify {boneName=dyn_shorts_in_l}
    scale {x=0.45}
   }
   n80 = "Transform (Modify) Bone - Bone: dyn_shorts_out_l"(ComponentPose=@n79, Alpha=@n77, Translation="0,0,0", Rotation="0,0,0", Scale="0.45,1,1") {
    rotationMode=BMM_Additive
    rotationSpace=BCS_BoneSpace
    scaleMode=BMM_Additive
    scaleSpace=BCS_BoneSpace
    translationMode=BMM_Additive
    translationSpace=BCS_BoneSpace
    alphaBoolBlend {blendOption=Linear}
    boneToModify {boneName=dyn_shorts_out_l}
    scale {x=0.45}
   }
   n81 = "Transform (Modify) Bone - Bone: dyn_shorts_ft_l"(ComponentPose=@n80, Alpha=@n77, Translation="0,0,0", Rotation="-30,0.000,0.000", Scale="0.8,1,1") {
    rotationMode=BMM_Additive
    rotationSpace=BCS_BoneSpace
    scaleMode=BMM_Additive
    scaleSpace=BCS_BoneSpace
    translationMode=BMM_Additive
    translationSpace=BCS_BoneSpace
    alphaBoolBlend {blendOption=Linear}
    boneToModify {boneName=dyn_shorts_ft_l}
    rotation {pitch=-30}
    scale {x=0.8}
   }
   Output Pose(Result=Component To Local(ComponentPose=@n81)) {layerGroup=None}
   Get legs_rot_sides()
  ]}
 EventGraph {nodes [
   Event BlueprintUpdateAnimation()
   > n1 = Sequence!(execute) -> (then_0, then_1, then_2:Exec, then_3:Exec, then_4:Exec)
   n2 = BreakTransform(InTransform=GetDeltaTransformFromRefPose(self=GetOwnerSkeletalMeshComponent(), BoneName=thigh_l, BaseName=pelvis))
   > Set thigh_up_l(thigh_up_l:double=MapRangeClamped(Value=BreakRotator(InRot=@n2.Rotation).Yaw, InRangeA=-20, InRangeB=-100)) -> (Output_Get:double)
   n3 = BreakTransform(InTransform=GetDeltaTransformFromRefPose(self=GetOwnerSkeletalMeshComponent(), BoneName=thigh_r, BaseName=pelvis))
   > Set thigh_up_r(thigh_up_r:double=MapRangeClamped(Value=BreakRotator(InRot=@n3.Rotation).Yaw, InRangeA=-20, InRangeB=-100)) -> (Output_Get:double)
   n4 = BreakTransform(InTransform=GetDeltaTransformFromRefPose(self=GetOwnerSkeletalMeshComponent(), BoneName=thigh_r, BaseName=thigh_l))
   > Set legs_rot_sides(legs_rot_sides:double=MapRangeClamped(Value=BreakRotator(InRot=@n4.Rotation).Pitch, InRangeA=0, InRangeB=16)) -> (Output_Get:double)
   n5 = BreakTransform(InTransform=GetDeltaTransformFromRefPose(self=GetOwnerSkeletalMeshComponent(), BoneName=thigh_r, BaseName=calf_r))
   > Set calf_rot_r(calf_rot_r:double=MapRangeClamped(Value=BreakRotator(InRot=@n5.Rotation).Yaw, InRangeA=-20, InRangeB=-120)) -> (Output_Get:double)
   n6 = BreakTransform(InTransform=GetDeltaTransformFromRefPose(self=GetOwnerSkeletalMeshComponent(), BoneName=thigh_l, BaseName=calf_l))
   > Set calf_rot_l(calf_rot_l:double=MapRangeClamped(Value=BreakRotator(InRot=@n6.Rotation).Yaw, InRangeA=-10, InRangeB=-120)) -> (Output_Get:double)
   n7 = GetOwnerSkeletalMeshComponent()
   n8 = BreakRotator(InRot=BreakTransform(InTransform=GetDeltaTransformFromRefPose(self=@n7, BoneName=upperarm_r, BaseName=spine_05)).Rotation)
   > Set arm_up_r(arm_up_r:double=MapRangeClamped(Value=@n8.Yaw, InRangeA=110, InRangeB=130, OutRangeA=1, OutRangeB=0)) -> (Output_Get:double)
   > Set arm_forward_r(arm_forward_r:double=MapRangeClamped(Value=@n8.Pitch, InRangeA=-15, InRangeB=-30)) -> (Output_Get:double)
   > Set ADS alpha(ADS alpha:double=float * float(A=Get arm_up_r() -> (arm_up_r:double), B=Get arm_forward_r() -> (arm_forward_r:double))) -> (Output_Get:double)
   n9 = GetPartSkeletalMeshComponent()
   > n10 = TransformFromBoneSpace!(
    execute
    self:Skinned Mesh Component Object Reference=@n9
    BoneName:Name=upperarm_twist_02_l
    InPosition:Vector="0, 0, 0"
    InRotation:Rotator="0, 0, 0"
   ) -> (then, OutPosition:Vector, OutRotation:Rotator)
   > n11 = TransformFromBoneSpace(self=@n9, BoneName=spine_05, InPosition="0, 0, 0", InRotation="0, 0, 0")
   n12 = "Distance (Vector)"!(V1:Vector=@n10, V2:Vector=@n11) -> (ReturnValue:double)
   n13 = MapRangeClamped(Value=@n12, InRangeA=15, InRangeB=20)
   n14 = MapRangeClamped(Value=@n12, InRangeA=30, InRangeB=25)
   > Set arm_forward_l(arm_forward_l:double=SelectFloat(A=@n14, B=@n13, bPickA="float < float"!(A:double=@n14, B:double=@n13) -> (ReturnValue:bool) {is=title K2Node_CallFunction})) -> (Output_Get:double)
   n15 = Get bIsCrouching()
   Set isNotCrouching(execute=@n1.then_1, isNotCrouching:double="To Float (Boolean)"(InBool=NOT Boolean(A=@n15))) -> (Output_Get:double)
   > n16 = Branch(Condition=Get bIsSwimSprinting() -> (bIsSwimSprinting:bool))
   > Set gravity_override(gravity_override="0, 0, 0")
   n17 = Branch(execute=@n16.else, Condition=Get bIsSkyDiving())
   > n18 = Branch(Condition=Get bIsParachuteOpened())
   Set gravity_override(execute=[@n17.else, @n18.then], gravity_override="0, 0,-980")
   Set gravity_override(execute=@n18.else, gravity_override="0, 0,-300")
   Branch(execute=@n1.then_2, Condition=@n15)
   > Set legs_rot_sides(legs_rot_sides=0.5)
   Branch(execute=@n1.then_3, Condition=Get bIsSliding())
   > Do Once(Reset=@n19.then)
   > Set from_sliding(from_sliding:bool=true) -> (Output_Get:bool)
   Branch(execute=@n1.then_4, Condition=Get from_sliding() -> (from_sliding:bool))
   > Branch(Condition=NOT Boolean(A=Get bIsSliding()))
   > Delay(Duration=0.15)
   > ResetDynamics(InTeleportType=TeleportPhysics)
   > ResetDynamics(InTeleportType=ResetPhysics)
   > n19 = Set from_sliding(from_sliding=false)
   n20 = Event OnBeginCrouching()
   ResetDynamics(execute=[@n22.then, @n20.then, @n21.then, @n23.then, @n24.then], InTeleportType=TeleportPhysics)
   > ResetDynamics(InTeleportType=ResetPhysics)
   n21 = Event OnEndCrouching()
   Event OnParachuteOpened()
   > n22 = Delay(Duration=1.5)
   n23 = Event OnEndSkydiving()
   Event BlueprintInitializeAnimation()
   > ResetDynamics(InTeleportType=TeleportPhysics)
   n24 = Event OnBeginSkydiving()
   MapRangeUnclamped!(Value:double=0, InRangeA:double=0, InRangeB:double=0, OutRangeA:double=0, OutRangeB:double=0) -> (ReturnValue:double)
  ]}
]
variables [
 ADS alpha {type=number}
 arm_forward_l {type=number}
 arm_forward_r {type=number}
 arm_up_r {type=number}
 calf_rot_l {type=number}
 calf_rot_r {type=number}
 from_sliding {type=boolean}
 gravity_override {type=Vector, default=<0 0 -980>}
 isNotCrouching {type=number}
 legs_rot_frtBk {type=number}
 legs_rot_sides {type=number}
 thigh_up_l {type=number}
 thigh_up_r {type=number}
]
