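# Physics asset for the hood part of the Future Samurai medium male player body, defining physical motion for its dyn_hood bones.
# Front, left, right and top hood bones each carry a capsule and are constrained directly to a small neck_02 capsule in a star layout rather than a chain.
# Tight swing and twist limits keep the hood close to its posed shape, with the larger top piece allowed the most movement.
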PhysicsAsset /BRCosmetics/Characters/Player/Male/Medium/Bodies/M_MED_Future_Samurai/Meshes/Parts/M_MED_Future_Samurai_dyn_hood
bodies {
 dyn_hood_ft [{shapeType=Capsule, length=5, radius=2, center=<-1 0 0.8>, rotation=<rot 90 0 0>}]
 dyn_hood_l [{shapeType=Capsule, length=5, radius=2, center=<2 2 -0.2>, rotation=<rot 90 0 0>}]
 dyn_hood_r [{shapeType=Capsule, length=5, radius=2, center=<-2 2 -0.2>, rotation=<rot 90 0 0>}]
 dyn_hood_top [{shapeType=Capsule, length=10, radius=3, center=<3 0 0>, rotation=<rot 90 0 0>}]
 neck_02 [{shapeType=Capsule, length=1, radius=1}]
}
constraints [
 {bone1Name=dyn_hood_r, bone2Name=neck_02, swing1LimitDegrees=2, swing2LimitDegrees=15, twistLimitDegrees=2}
 {bone1Name=dyn_hood_ft, bone2Name=neck_02, swing1LimitDegrees=5, swing2LimitDegrees=15, twistLimitDegrees=2}
 {bone1Name=dyn_hood_l, bone2Name=neck_02, swing1LimitDegrees=2, swing2LimitDegrees=15, twistLimitDegrees=2}
 {bone1Name=dyn_hood_top, bone2Name=neck_02, swing1LimitDegrees=15, swing2LimitDegrees=25, twistLimitDegrees=5}
]
